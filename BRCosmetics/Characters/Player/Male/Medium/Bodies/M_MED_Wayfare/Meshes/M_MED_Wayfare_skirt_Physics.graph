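# Physics asset for the male medium Wayfare body's skirt, defining three rear cloth bone chains: center, left and right (spelled "roincloth").
# Each chain hangs from a pelvis-linked MSTR bone through three capsules with limited swing and twist, tighter on the right.
# Free-angle cross constraints tie the lower two side segments to the center chain.
# Pelvis, leg and right-foot capsules (no left foot) plus Floor_COL spheres provide body and floor collision shapes.
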PhysicsAsset /BRCosmetics/Characters/Player/Male/Medium/Bodies/M_MED_Wayfare/Meshes/M_MED_Wayfare_skirt_Physics
bodies {
 Floor_COL [
  {shapeType=Sphere, radius=183.4375, shapeName=root_sphere, center=<0 -8.6403 -175.5073>}
  {shapeType=Sphere, radius=42.0132, shapeName=root_sphere, center=<0.0002 2.4787 2.7666>}
 ]
 calf_l [{shapeType=Capsule, length=31.988, radius=14.3954, shapeName=dyn_holster_top_MSTR_capsule, center=<-27.9785 0.6199 0>, rotation=<rot 90 0 0>}]
 calf_r [{shapeType=Capsule, length=33.8411, radius=12.8867, shapeName=dyn_holster_top_MSTR_capsule, center=<26.3007 -0.4915 0>, rotation=<rot 90 0 0>}]
 dyn_loincloth_bk_01 [{shapeType=Capsule, length=12, radius=9, shapeName=dyn_loincloth_bk_01_capsule, center=<7 0 8>, rotation=<rot 0 0 95>}]
 dyn_loincloth_bk_02 [{shapeType=Capsule, length=11.5, radius=9.5, shapeName=dyn_loincloth_bk_02_capsule, center=<7 0 8>, rotation=<rot 0 0 95>}]
 dyn_loincloth_bk_03 [{shapeType=Capsule, length=11.5, radius=9.5, shapeName=dyn_loincloth_bk_03_capsule, center=<7 0 8>, rotation=<rot 0 0 95>}]
 dyn_loincloth_bk_MSTR [{shapeType=Capsule, length=0.505, radius=0.505, shapeName=dyn_loincloth_bk_MSTR_capsule, rotation=<rot 0 -90 0>}]
 dyn_loincloth_bk_l_01 [{shapeType=Capsule, length=12, radius=9, shapeName=dyn_loincloth_bk_l_01_capsule, center=<7 0 8>, rotation=<rot -0.0001 0 95>}]
 dyn_loincloth_bk_l_02 [{shapeType=Capsule, length=12, radius=9, shapeName=dyn_loincloth_bk_l_02_capsule, center=<7 0 8>, rotation=<rot 0 0 90>}]
 dyn_loincloth_bk_l_03 [{shapeType=Capsule, length=12, radius=9, shapeName=dyn_loincloth_bk_l_03_capsule, center=<7 0 8>, rotation=<rot 0 0 90>}]
 dyn_loincloth_bk_l_MSTR [{shapeType=Capsule, length=0.505, radius=0.505, shapeName=dyn_loincloth_bk_l_MSTR_capsule, rotation=<rot 0 -90 0>}]
 dyn_roincloth_bk_r_01 [{shapeType=Capsule, length=10.5, radius=10.5, shapeName=dyn_roincloth_bk_r_01_capsule, center=<8.5 0 9.5>, rotation=<rot 0 -0.0001 86.6852>}]
 dyn_roincloth_bk_r_02 [{shapeType=Capsule, length=10.5, radius=10.5, shapeName=dyn_roincloth_bk_r_02_capsule, center=<7 0 9>, rotation=<rot -0.0001 -0.0004 85.2271>}]
 dyn_roincloth_bk_r_03 [{
   shapeType=Capsule
   length=10.5
   radius=10.5
   shapeName=dyn_roincloth_bk_r_03_capsule
   center=<6.5 0 8>
   rotation=<rot -0.0001 -0.0004 88.3007>
  }]
 dyn_roincloth_bk_r_MSTR [{shapeType=Capsule, length=0.505, radius=0.505, shapeName=dyn_roincloth_bk_r_MSTR_capsule, rotation=<rot 0 -90 0>}]
 foot_r [{
   shapeType=Capsule
   length=11.5037
   radius=14.4266
   shapeName=foot_r_capsule
   center=<3.9857 -6.9894 -0.2675>
   rotation=<rot -1.9473 -170.9849 -85.8186>
  }]
 pelvis [{shapeType=Capsule, length=10.0244, radius=13.1221, shapeName=dyn_holster_top_MSTR_capsule, center=<-5.2584 3.0551 0>}]
 thigh_l [{
   shapeType=Capsule
   length=24.5409
   radius=12.0731
   shapeName=dyn_holster_top_MSTR_capsule
   center=<-20.4747 1.7656 0.0001>
   rotation=<rot 90 0 0>
  }]
 thigh_r [{
   shapeType=Capsule
   length=24.5242
   radius=12.0899
   shapeName=dyn_holster_top_MSTR_capsule
   center=<20.472 -1.7835 -0.6637>
   rotation=<rot 89.4188 179.9973 -0.0022>
  }]
}
constraints [
 {
  bone1Name=dyn_loincloth_bk_MSTR
  bone2Name=pelvis
  swing1LimitDegrees=45
  swing1Motion=Free
  swing2LimitDegrees=45
  swing2Motion=Free
  twistLimitDegrees=45
  twistMotion=Free
 }
 {
  bone1Name=dyn_loincloth_bk_l_MSTR
  bone2Name=pelvis
  swing1LimitDegrees=45
  swing1Motion=Free
  swing2LimitDegrees=45
  swing2Motion=Free
  twistLimitDegrees=45
  twistMotion=Free
 }
 {
  bone1Name=dyn_roincloth_bk_r_MSTR
  bone2Name=pelvis
  swing1LimitDegrees=45
  swing1Motion=Free
  swing2LimitDegrees=45
  swing2Motion=Free
  twistLimitDegrees=45
  twistMotion=Free
 }
 {bone1Name=dyn_loincloth_bk_01, bone2Name=dyn_loincloth_bk_MSTR, swing1LimitDegrees=35, swing2LimitDegrees=45, twistLimitDegrees=15}
 {bone1Name=dyn_loincloth_bk_03, bone2Name=dyn_loincloth_bk_02, swing1LimitDegrees=35, swing2LimitDegrees=45, twistLimitDegrees=15}
 {bone1Name=dyn_loincloth_bk_02, bone2Name=dyn_loincloth_bk_01, swing1LimitDegrees=35, swing2LimitDegrees=45, twistLimitDegrees=15}
 {bone1Name=dyn_loincloth_bk_l_01, bone2Name=dyn_loincloth_bk_l_MSTR, swing1LimitDegrees=35, swing2LimitDegrees=45, twistLimitDegrees=15}
 {bone1Name=dyn_loincloth_bk_l_02, bone2Name=dyn_loincloth_bk_l_01, swing1LimitDegrees=35, swing2LimitDegrees=45, twistLimitDegrees=15}
 {bone1Name=dyn_loincloth_bk_l_03, bone2Name=dyn_loincloth_bk_l_02, swing1LimitDegrees=35, swing2LimitDegrees=45, twistLimitDegrees=15}
 {bone1Name=dyn_roincloth_bk_r_01, bone2Name=dyn_roincloth_bk_r_MSTR, swing1LimitDegrees=10, swing2LimitDegrees=45, twistLimitDegrees=10}
 {bone1Name=dyn_roincloth_bk_r_02, bone2Name=dyn_roincloth_bk_r_01, swing1LimitDegrees=10, swing2LimitDegrees=45, twistLimitDegrees=10}
 {bone1Name=dyn_roincloth_bk_r_03, bone2Name=dyn_roincloth_bk_r_02, swing1LimitDegrees=10, swing2LimitDegrees=45, twistLimitDegrees=10}
 {
  bone1Name=thigh_l
  bone2Name=pelvis
  swing1LimitDegrees=45
  swing1Motion=Free
  swing2LimitDegrees=45
  swing2Motion=Free
  twistLimitDegrees=45
  twistMotion=Free
 }
 {
  bone1Name=thigh_r
  bone2Name=pelvis
  swing1LimitDegrees=45
  swing1Motion=Free
  swing2LimitDegrees=45
  swing2Motion=Free
  twistLimitDegrees=45
  twistMotion=Free
 }
 {
  bone1Name=calf_l
  bone2Name=thigh_l
  swing1LimitDegrees=45
  swing1Motion=Free
  swing2LimitDegrees=45
  swing2Motion=Free
  twistLimitDegrees=45
  twistMotion=Free
 }
 {
  bone1Name=calf_r
  bone2Name=thigh_r
  swing1LimitDegrees=45
  swing1Motion=Free
  swing2LimitDegrees=45
  swing2Motion=Free
  twistLimitDegrees=45
  twistMotion=Free
 }
 {
  bone1Name=foot_r
  bone2Name=calf_r
  swing1LimitDegrees=45
  swing1Motion=Free
  swing2LimitDegrees=45
  swing2Motion=Free
  twistLimitDegrees=45
  twistMotion=Free
 }
 {
  bone1Name=dyn_loincloth_bk_l_02
  bone2Name=dyn_loincloth_bk_02
  swing1LimitDegrees=45
  swing1Motion=Free
  swing2LimitDegrees=45
  swing2Motion=Free
  twistLimitDegrees=45
  twistMotion=Free
 }
 {
  bone1Name=dyn_roincloth_bk_r_02
  bone2Name=dyn_loincloth_bk_02
  swing1LimitDegrees=45
  swing1Motion=Free
  swing2LimitDegrees=45
  swing2Motion=Free
  twistLimitDegrees=45
  twistMotion=Free
 }
 {
  bone1Name=dyn_loincloth_bk_l_03
  bone2Name=dyn_loincloth_bk_03
  swing1LimitDegrees=45
  swing1Motion=Free
  swing2LimitDegrees=45
  swing2Motion=Free
  twistLimitDegrees=45
  twistMotion=Free
 }
 {
  bone1Name=dyn_roincloth_bk_r_03
  bone2Name=dyn_loincloth_bk_03
  swing1LimitDegrees=45
  swing1Motion=Free
  swing2LimitDegrees=45
  swing2Motion=Free
  twistLimitDegrees=45
  twistMotion=Free
 }
]
