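# Physics asset for the medium male player Wayfare body, built around three dynamic rear loincloth bone chains: center, left and right (named "roincloth").
# Each chain runs three capsules from a tiny master capsule freely constrained to the pelvis, with narrower angle limits along the chain.
# Pelvis, thigh, calf and right-foot capsules plus two root spheres, one very large and offset, form the body shapes, linked by free constraints; the left foot has no body.
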PhysicsAsset /Game/Characters/Player/Male/Medium/Bodies/M_MED_Wayfare/Meshes/M_MED_Wayfare_Physics
bodies {
 calf_l [{shapeType=Capsule, length=39.113, radius=20.3954, shapeName=dyn_holster_top_MSTR_capsule, center=<-29.9666 0 0>, rotation=<rot 90 0 0>}]
 calf_r [{shapeType=Capsule, length=29.2161, radius=13.0117, shapeName=dyn_holster_top_MSTR_capsule, center=<26.3007 0 0>, rotation=<rot 90 0 0>}]
 dyn_loincloth_bk_01 [{shapeType=Capsule, length=19.6285, radius=29.52, shapeName=dyn_holster_top_capsule, center=<9.7869 0 20>, rotation=<rot 0 0 90>}]
 dyn_loincloth_bk_02 [{shapeType=Capsule, length=70.6998, radius=29.52, shapeName=dyn_holster_top_capsule, center=<9.7869 0 20>, rotation=<rot 0 0 90>}]
 dyn_loincloth_bk_03 [{shapeType=Capsule, length=70.6998, radius=29.52, shapeName=dyn_holster_top_capsule, center=<9.7869 0 20>, rotation=<rot 0 0 90>}]
 dyn_loincloth_bk_MSTR [{shapeType=Capsule, length=0.505, radius=0.505, shapeName=dyn_loincloth_bk_MSTR_capsule, rotation=<rot 0 -90 0>}]
 dyn_loincloth_bk_l_01 [{shapeType=Capsule, length=21.1508, radius=27.9977, shapeName=dyn_holster_top_capsule, center=<9.7869 0 30>, rotation=<rot 0 0 90>}]
 dyn_loincloth_bk_l_02 [{shapeType=Capsule, length=21.1508, radius=27.9977, shapeName=dyn_holster_top_capsule, center=<9.7869 0 30>, rotation=<rot 0 0 90>}]
 dyn_loincloth_bk_l_03 [{shapeType=Capsule, length=21.1508, radius=27.9977, shapeName=dyn_holster_top_capsule, center=<9.7869 0 30>, rotation=<rot 0 0 90>}]
 dyn_loincloth_bk_l_MSTR [{shapeType=Capsule, length=0.505, radius=0.505, shapeName=dyn_loincloth_bk_l_MSTR_capsule, rotation=<rot 0 -90 0>}]
 dyn_roincloth_bk_r_01 [{shapeType=Capsule, length=21.1508, radius=27.9977, shapeName=dyn_holster_top_capsule, center=<9.7869 0 30>, rotation=<rot 0 0 90>}]
 dyn_roincloth_bk_r_02 [{shapeType=Capsule, length=21.1508, radius=27.9977, shapeName=dyn_holster_top_capsule, center=<9.7869 0 30>, rotation=<rot 0 0 90>}]
 dyn_roincloth_bk_r_03 [{shapeType=Capsule, length=21.1508, radius=27.9977, shapeName=dyn_holster_top_capsule, center=<9.7869 0 30>, rotation=<rot 0 0 90>}]
 dyn_roincloth_bk_r_MSTR [{shapeType=Capsule, length=0.505, radius=0.505, shapeName=dyn_roincloth_bk_r_MSTR_capsule, rotation=<rot 0 -90 0>}]
 foot_r [{shapeType=Capsule, length=4.1509, radius=25.6223, shapeName=foot_r_capsule, rotation=<rot 0 0 90>}]
 pelvis [{shapeType=Capsule, length=10.7744, radius=12.7471, shapeName=dyn_holster_top_MSTR_capsule, center=<-10.7828 0.2908 0>}]
 root [
  {shapeType=Sphere, radius=183.4375, shapeName=root_sphere, center=<0 -8.6403 -175.5073>}
  {shapeType=Sphere, radius=39.7632, shapeName=root_sphere, center=<0.0002 2.4787 8.5972>}
 ]
 thigh_l [{
   shapeType=Capsule
   length=24.5409
   radius=12.0731
   shapeName=dyn_holster_top_MSTR_capsule
   center=<-20.4747 0.0011 0.0001>
   rotation=<rot 90 0 0>
  }]
 thigh_r [{
   shapeType=Capsule
   length=24.5242
   radius=12.0899
   shapeName=dyn_holster_top_MSTR_capsule
   center=<20.4747 -0.0011 -0.0001>
   rotation=<rot 90 0 180>
  }]
}
constraints [
 {
  bone1Name=dyn_loincloth_bk_MSTR
  bone2Name=pelvis
  swing1LimitDegrees=45
  swing1Motion=Free
  swing2LimitDegrees=45
  swing2Motion=Free
  twistLimitDegrees=45
  twistMotion=Free
 }
 {
  bone1Name=dyn_loincloth_bk_l_MSTR
  bone2Name=pelvis
  swing1LimitDegrees=45
  swing1Motion=Free
  swing2LimitDegrees=45
  swing2Motion=Free
  twistLimitDegrees=45
  twistMotion=Free
 }
 {
  bone1Name=dyn_roincloth_bk_r_MSTR
  bone2Name=pelvis
  swing1LimitDegrees=45
  swing1Motion=Free
  swing2LimitDegrees=45
  swing2Motion=Free
  twistLimitDegrees=45
  twistMotion=Free
 }
 {bone1Name=dyn_loincloth_bk_01, bone2Name=dyn_loincloth_bk_MSTR, swing1LimitDegrees=10, swing2LimitDegrees=45, twistLimitDegrees=10}
 {bone1Name=dyn_loincloth_bk_03, bone2Name=dyn_loincloth_bk_02, swing1LimitDegrees=10, swing2LimitDegrees=45, twistLimitDegrees=10}
 {bone1Name=dyn_loincloth_bk_02, bone2Name=dyn_loincloth_bk_01, swing1LimitDegrees=10, swing2LimitDegrees=45, twistLimitDegrees=10}
 {bone1Name=dyn_loincloth_bk_l_01, bone2Name=dyn_loincloth_bk_l_MSTR, swing1LimitDegrees=20, swing2LimitDegrees=45, twistLimitDegrees=20}
 {bone1Name=dyn_loincloth_bk_l_02, bone2Name=dyn_loincloth_bk_l_01, swing1LimitDegrees=20, swing2LimitDegrees=45, twistLimitDegrees=20}
 {bone1Name=dyn_loincloth_bk_l_03, bone2Name=dyn_loincloth_bk_l_02, swing1LimitDegrees=20, swing2LimitDegrees=45, twistLimitDegrees=20}
 {bone1Name=dyn_roincloth_bk_r_01, bone2Name=dyn_roincloth_bk_r_MSTR, swing1LimitDegrees=10, swing2LimitDegrees=45, twistLimitDegrees=10}
 {bone1Name=dyn_roincloth_bk_r_02, bone2Name=dyn_roincloth_bk_r_01, swing1LimitDegrees=10, swing2LimitDegrees=45, twistLimitDegrees=10}
 {bone1Name=dyn_roincloth_bk_r_03, bone2Name=dyn_roincloth_bk_r_02, swing1LimitDegrees=10, swing2LimitDegrees=45, twistLimitDegrees=10}
 {
  bone1Name=thigh_l
  bone2Name=pelvis
  swing1LimitDegrees=45
  swing1Motion=Free
  swing2LimitDegrees=45
  swing2Motion=Free
  twistLimitDegrees=45
  twistMotion=Free
 }
 {
  bone1Name=thigh_r
  bone2Name=pelvis
  swing1LimitDegrees=45
  swing1Motion=Free
  swing2LimitDegrees=45
  swing2Motion=Free
  twistLimitDegrees=45
  twistMotion=Free
 }
 {
  bone1Name=calf_l
  bone2Name=thigh_l
  swing1LimitDegrees=45
  swing1Motion=Free
  swing2LimitDegrees=45
  swing2Motion=Free
  twistLimitDegrees=45
  twistMotion=Free
 }
 {
  bone1Name=calf_r
  bone2Name=thigh_r
  swing1LimitDegrees=45
  swing1Motion=Free
  swing2LimitDegrees=45
  swing2Motion=Free
  twistLimitDegrees=45
  twistMotion=Free
 }
 {
  bone1Name=pelvis
  bone2Name=root
  swing1LimitDegrees=45
  swing1Motion=Free
  swing2LimitDegrees=45
  swing2Motion=Free
  twistLimitDegrees=45
  twistMotion=Free
 }
 {
  bone1Name=foot_r
  bone2Name=calf_r
  swing1LimitDegrees=45
  swing1Motion=Free
  swing2LimitDegrees=45
  swing2Motion=Free
  twistLimitDegrees=45
  twistMotion=Free
 }
]
