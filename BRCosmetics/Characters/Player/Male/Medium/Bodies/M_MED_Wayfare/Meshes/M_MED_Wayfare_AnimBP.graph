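# Post-process style AnimBP for the Wayfare male body cosmetic, correcting and simulating its loincloth, shoulder pads and neck guard.
# The copied parent pose runs through two Control Rigs and state-driven additive bone offsets, then a pelvis-space skirt RigidBody with overridable gravity.
# The event graph derives those states from weapon, movement, crouch, jump and glider data.
# A second simulation chain blending root- and pelvis-space RigidBodies is not wired to the output.
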AnimBlueprint /BRCosmetics/Characters/Player/Male/Medium/Bodies/M_MED_Wayfare/Meshes/M_MED_Wayfare_AnimBP
target_skeleton=&/M_MED_Wayfare_Skeleton
graphs [
 AnimGraph {nodes [
   n1 = RigidBody!(ComponentPose, Alpha, OverrideWorldGravity:Vector="0,0,0", ExternalForce, ComponentLinearAccScale:Vector="0.5,0.3,0.5") -> (Pose) {
    bClampLinearTranslationLimitToRefPose=true
    bOverrideWorldGravity=true
    overridePhysicsAsset=/Game/Characters/Player/Male/Medium/Bodies/M_MED_Wayfare/Meshes/M_MED_Wayfare_Physics
    simulationSpace=BaseBoneSpace
    alphaBoolBlend {blendOption=Linear}
    baseBoneRef {boneName=root}
    componentAppliedLinearAccClamp=<2000 2000 6000>
    componentLinearAccScale {z=0.45}
    componentLinearVelScale=<0.8 0.3 1>
    overrideWorldGravity {z=-980}
    simSpaceSettings {
     maxAngularAcceleration=1000
     maxAngularVelocity=3
     maxLinearAcceleration=700
     maxLinearVelocity=500
     velocityScaleZ=0
     externalLinearDragV=<1 1>
    }
   }
   n2 = Component To Local(ComponentPose=@n1)
   n3 = Get NewSliding() -> (NewSliding:bool)
   n4 = "Transform (Modify) Bone - Bone: dyn_loincloth_bk_l_MSTR"!(
    ComponentPose=Local To Component(LocalPose=Use cached pose 'PreSim'())
    bAlphaBoolEnabled:bool=@n3
    Translation="0,0,-7"
    Rotation="90,0,0"
    Scale="1,1,1"
   ) -> (Pose) {
    alphaInputType=Bool
    rotationMode=BMM_Additive
    rotationSpace=BCS_BoneSpace
    scaleMode=BMM_Additive
    scaleSpace=BCS_BoneSpace
    translationMode=BMM_Additive
    translationSpace=BCS_BoneSpace
    alphaBoolBlend {blendInTime=0.1, blendOption=Linear, blendOutTime=0.1}
    boneToModify {boneName=dyn_loincloth_bk_l_MSTR}
    rotation {pitch=90}
    translation {z=-7}
   }
   n5 = "Transform (Modify) Bone - Bone: dyn_loincloth_bk_MSTR"!(ComponentPose=@n4, bAlphaBoolEnabled:bool=@n3, Translation="0,0,-7", Rotation="90,0,0", Scale="1,1,1") -> (Pose) {
    alphaInputType=Bool
    rotationMode=BMM_Additive
    rotationSpace=BCS_BoneSpace
    scaleMode=BMM_Additive
    scaleSpace=BCS_BoneSpace
    translationMode=BMM_Additive
    translationSpace=BCS_BoneSpace
    alphaBoolBlend {blendInTime=0.1, blendOption=Linear, blendOutTime=0.1}
    boneToModify {boneName=dyn_loincloth_bk_MSTR}
    rotation {pitch=90}
    translation {z=-7}
   }
   n6 = "Transform (Modify) Bone - Bone: dyn_roincloth_bk_r_MSTR"!(ComponentPose=@n5, bAlphaBoolEnabled:bool=@n3, Translation="0,0,-7", Rotation="90,0,0", Scale="1,1,1") -> (Pose) {
    alphaInputType=Bool
    rotationMode=BMM_Additive
    rotationSpace=BCS_BoneSpace
    scaleMode=BMM_Additive
    scaleSpace=BCS_BoneSpace
    translationMode=BMM_Additive
    translationSpace=BCS_BoneSpace
    alphaBoolBlend {blendInTime=0.1, blendOption=Linear, blendOutTime=0.1}
    boneToModify {boneName=dyn_roincloth_bk_r_MSTR}
    rotation {pitch=90}
    translation {z=-7}
   }
   n7 = RigidBody(ComponentPose=@n6, OverrideWorldGravity="0,0,0", ComponentLinearAccScale="0.6,0.3,0.6") {
    bClampLinearTranslationLimitToRefPose=true
    bOverrideWorldGravity=true
    overridePhysicsAsset=&M_MED_Wayfare_Physics
    simulationSpace=BaseBoneSpace
    alphaBoolBlend {blendOption=Linear}
    baseBoneRef {boneName=pelvis}
    componentAppliedLinearAccClamp=<5000 5000 5000>
    componentLinearAccScale {z=0.45}
    componentLinearVelScale=<0.6 0.3 0.6>
    overrideWorldGravity {z=-980}
    simSpaceSettings {
     maxAngularAcceleration=1000
     maxAngularVelocity=3
     maxLinearAcceleration=700
     maxLinearVelocity=500
     velocityScaleZ=0
     worldAlpha=0.4
     externalLinearDragV=<1 1>
    }
   }
   n8 = Component To Local(ComponentPose=@n7)
   n9 = Get RootSim() -> (RootSim:bool)
   n10 = Get RootSim()
   n11 = "Transform (Modify) Bone - Bone: dyn_loincloth_bk_l_MSTR"(
    ComponentPose=Local To Component(LocalPose=Blend Poses by bool(BlendPose_0=@n2, BlendPose_1=@n8, BlendTime_0=2, BlendTime_1=1, bActiveValue=@n9) {blendType=Linear})
    bAlphaBoolEnabled=@n10
    Translation="0,0,3"
    Rotation="0,0,0"
    Scale="1,1,1"
   ) {
    alphaInputType=Bool
    rotationMode=BMM_Additive
    rotationSpace=BCS_BoneSpace
    scaleMode=BMM_Additive
    scaleSpace=BCS_BoneSpace
    translationMode=BMM_Additive
    translationSpace=BCS_BoneSpace
    alphaBoolBlend {blendInTime=0.1, blendOption=Linear, blendOutTime=0.1}
    boneToModify {boneName=dyn_loincloth_bk_l_MSTR}
    translation {z=3}
   }
   n12 = "Transform (Modify) Bone - Bone: dyn_loincloth_bk_MSTR"(ComponentPose=@n11, bAlphaBoolEnabled=@n10, Translation="0,0,3", Rotation="0,0,0", Scale="1,1,1") {
    alphaInputType=Bool
    rotationMode=BMM_Additive
    rotationSpace=BCS_BoneSpace
    scaleMode=BMM_Additive
    scaleSpace=BCS_BoneSpace
    translationMode=BMM_Additive
    translationSpace=BCS_BoneSpace
    alphaBoolBlend {blendInTime=0.1, blendOption=Linear, blendOutTime=0.1}
    boneToModify {boneName=dyn_loincloth_bk_MSTR}
    translation {z=3}
   }
   "Transform (Modify) Bone - Bone: dyn_roincloth_bk_r_MSTR"(ComponentPose=@n12, bAlphaBoolEnabled=@n10, Translation="0,0,3", Rotation="0,0,0", Scale="1,1,1") {
    alphaInputType=Bool
    rotationMode=BMM_Additive
    rotationSpace=BCS_BoneSpace
    scaleMode=BMM_Additive
    scaleSpace=BCS_BoneSpace
    translationMode=BMM_Additive
    translationSpace=BCS_BoneSpace
    alphaBoolBlend {blendInTime=0.1, blendOption=Linear, blendOutTime=0.1}
    boneToModify {boneName=dyn_roincloth_bk_r_MSTR}
    translation {z=3}
   }
   n13 = RigidBody!(ComponentPose=Local To Component(LocalPose=Use cached pose 'PreSim'()), Alpha, OverrideWorldGravity=Get GravityOverride(), ExternalForce) -> (Pose) {
    bClampLinearTranslationLimitToRefPose=true
    bOverrideWorldGravity=true
    overridePhysicsAsset=&/M_MED_Wayfare_skirt_Physics
    simulationSpace=BaseBoneSpace
    alphaBoolBlend {blendOption=Linear}
    baseBoneRef {boneName=pelvis}
    componentAppliedLinearAccClamp=<3000 3000 3000>
    componentLinearAccScale=<0.3 0.3 0.6>
    componentLinearVelScale=<0.3 0.3 0.6>
    simSpaceSettings {
     maxAngularAcceleration=1000
     maxAngularVelocity=3
     maxLinearAcceleration=700
     maxLinearVelocity=500
     velocityScaleZ=0
     worldAlpha=1
     externalLinearDragV=<0.2 0.2>
    }
   }
   Output Pose(Result=Component To Local(ComponentPose=@n13)) {layerGroup=None}
   n14 = Control Rig(Source=Copy Pose From Mesh() {bUseAttachedParent=true}) {
    alphaBoolBlend {blendOption=Linear}
    controlRigAssetReference {blueprintRigClass=&/M_MED_DeformRig_Wayfare_ControlRig.M_MED_DeformRig_Wayfare_ControlRig_C}
   }
   n15 = Control Rig!(Source=@n14, Alpha, movement:double=Get movement() -> (movement:double)) -> (Pose) {
    alphaBoolBlend {blendOption=Linear}
    controlRigAssetReference {blueprintRigClass=&/M_MED_Wayfare_CtrlRig.M_MED_Wayfare_CtrlRig_C}
   }
   n16 = Get bIsMoving2D()
   n17 = "Transform (Modify) Bone - Bone: dyn_loincloth_bk_l_MSTR"(ComponentPose=Local To Component(LocalPose=@n15), bAlphaBoolEnabled=@n16, Translation="0,0,0", Rotation="15,0,0.000", Scale="1,1,1") {
    alphaInputType=Bool
    rotationMode=BMM_Additive
    rotationSpace=BCS_BoneSpace
    alphaBoolBlend {blendInTime=0.1, blendOption=Linear, blendOutTime=0.5}
    alphaScaleBiasClamp {bInterpResult=true}
    boneToModify {boneName=dyn_loincloth_bk_l_MSTR}
    rotation {pitch=15}
   }
   n18 = "Transform (Modify) Bone - Bone: dyn_loincloth_bk_MSTR"(ComponentPose=@n17, bAlphaBoolEnabled=@n16, Translation="0,0,0", Rotation="15,0,0.000", Scale="1,1,1") {
    alphaInputType=Bool
    rotationMode=BMM_Additive
    rotationSpace=BCS_BoneSpace
    alphaBoolBlend {blendInTime=0.1, blendOption=Linear, blendOutTime=0.5}
    alphaScaleBiasClamp {bInterpResult=true}
    boneToModify {boneName=dyn_loincloth_bk_MSTR}
    rotation {pitch=15}
   }
   n19 = Get jumping() -> (jumping:double)
   n20 = "Transform (Modify) Bone - Bone: dyn_loincloth_bk_l_MSTR"!(ComponentPose=@n18, Alpha:float=@n19, Translation="0,0,0", Rotation="15,15,0.000", Scale="1,1,1") -> (Pose) {
    rotationMode=BMM_Additive
    rotationSpace=BCS_BoneSpace
    alphaBoolBlend {blendInTime=0.1, blendOption=Linear, blendOutTime=0.5}
    alphaScaleBiasClamp {bInterpResult=true}
    boneToModify {boneName=dyn_loincloth_bk_l_MSTR}
    rotation {pitch=15, yaw=15}
   }
   n21 = "Transform (Modify) Bone - Bone: dyn_loincloth_bk_MSTR"!(ComponentPose=@n20, Alpha:float=@n19, Translation="0,0,0", Rotation="15,25,0.000", Scale="1,1,1") -> (Pose) {
    rotationMode=BMM_Additive
    rotationSpace=BCS_BoneSpace
    alphaBoolBlend {blendInTime=0.1, blendOption=Linear, blendOutTime=0.5}
    alphaScaleBiasClamp {bInterpResult=true}
    boneToModify {boneName=dyn_loincloth_bk_MSTR}
    rotation {pitch=15, yaw=25}
   }
   n22 = "Transform (Modify) Bone - Bone: dyn_roincloth_bk_r_MSTR"!(ComponentPose=@n21, Alpha:float=@n19, Translation="-3,-3,0", Rotation="15,30,0.000", Scale="1,1,1") -> (Pose) {
    rotationMode=BMM_Additive
    rotationSpace=BCS_BoneSpace
    translationMode=BMM_Additive
    translationSpace=BCS_BoneSpace
    alphaBoolBlend {blendInTime=0.1, blendOption=Linear, blendOutTime=0.5}
    alphaScaleBiasClamp {bInterpResult=true}
    boneToModify {boneName=dyn_roincloth_bk_r_MSTR}
    rotation {pitch=15, yaw=30}
    translation=<-3 -3>
   }
   n23 = "Transform (Modify) Bone - Bone: Floor_COL"(ComponentPose=@n22, Alpha=@n19, Translation="0,0,-100", Rotation="0,0,0", Scale="1,1,1") {
    rotationMode=BMM_Additive
    rotationSpace=BCS_BoneSpace
    translationMode=BMM_Additive
    translationSpace=BCS_BoneSpace
    alphaBoolBlend {blendInTime=0.1, blendOption=Linear, blendOutTime=0.5}
    boneToModify {boneName=Floor_COL}
    translation {z=-100}
   }
   n24 = Get still_crouch() -> (still_crouch:bool)
   n25 = "Transform (Modify) Bone - Bone: dyn_loincloth_bk_01"!(ComponentPose=@n23, bAlphaBoolEnabled:bool=@n24, Translation="0,0,0", Rotation="45,-0.000,0.000", Scale="1,1,1") -> (Pose) {
    alphaInputType=Bool
    rotationMode=BMM_Additive
    rotationSpace=BCS_BoneSpace
    scaleMode=BMM_Additive
    scaleSpace=BCS_BoneSpace
    translationMode=BMM_Additive
    translationSpace=BCS_BoneSpace
    alphaBoolBlend {blendInTime=0.1, blendOption=Linear, blendOutTime=0.5}
    boneToModify {boneName=dyn_loincloth_bk_01}
    rotation {pitch=45}
   }
   n26 = "Transform (Modify) Bone - Bone: dyn_loincloth_bk_02"!(ComponentPose=@n25, bAlphaBoolEnabled:bool=@n24, Translation="0,0,0", Rotation="20,-0.000,0.000", Scale="1,1,1") -> (Pose) {
    alphaInputType=Bool
    rotationMode=BMM_Additive
    rotationSpace=BCS_BoneSpace
    scaleMode=BMM_Additive
    scaleSpace=BCS_BoneSpace
    translationMode=BMM_Additive
    translationSpace=BCS_BoneSpace
    alphaBoolBlend {blendInTime=0.1, blendOption=Linear, blendOutTime=0.5}
    boneToModify {boneName=dyn_loincloth_bk_02}
    rotation {pitch=20}
   }
   n27 = "Transform (Modify) Bone - Bone: dyn_loincloth_bk_l_01"!(ComponentPose=@n26, bAlphaBoolEnabled:bool=@n24, Translation="0,0,0", Rotation="15,-0.000,0.000", Scale="1,1,1") -> (Pose) {
    alphaInputType=Bool
    rotationMode=BMM_Additive
    rotationSpace=BCS_BoneSpace
    scaleMode=BMM_Additive
    scaleSpace=BCS_BoneSpace
    translationMode=BMM_Additive
    translationSpace=BCS_BoneSpace
    alphaBoolBlend {blendInTime=0.1, blendOption=Linear, blendOutTime=0.5}
    boneToModify {boneName=dyn_loincloth_bk_l_01}
    rotation {pitch=15}
   }
   n28 = "Transform (Modify) Bone - Bone: dyn_loincloth_bk_l_02"!(ComponentPose=@n27, bAlphaBoolEnabled:bool=@n24, Translation="0,0,0", Rotation="20,-0.000,35", Scale="1,1,1") -> (Pose) {
    alphaInputType=Bool
    rotationMode=BMM_Additive
    rotationSpace=BCS_BoneSpace
    scaleMode=BMM_Additive
    scaleSpace=BCS_BoneSpace
    translationMode=BMM_Additive
    translationSpace=BCS_BoneSpace
    alphaBoolBlend {blendInTime=0.1, blendOption=Linear, blendOutTime=0.5}
    boneToModify {boneName=dyn_loincloth_bk_l_02}
    rotation {pitch=20, roll=35}
   }
   n29 = "Transform (Modify) Bone - Bone: dyn_roincloth_bk_r_01"!(ComponentPose=@n28, bAlphaBoolEnabled:bool=@n24, Translation="0,0,0", Rotation="0,15,0.000", Scale="1,1,1") -> (Pose) {
    alphaInputType=Bool
    rotationMode=BMM_Additive
    rotationSpace=BCS_BoneSpace
    scaleMode=BMM_Additive
    scaleSpace=BCS_BoneSpace
    translationMode=BMM_Additive
    translationSpace=BCS_BoneSpace
    alphaBoolBlend {blendInTime=0.1, blendOption=Linear, blendOutTime=0.5}
    boneToModify {boneName=dyn_roincloth_bk_r_01}
    rotation {yaw=15}
   }
   n30 = "Transform (Modify) Bone - Bone: dyn_roincloth_bk_r_02"!(ComponentPose=@n29, bAlphaBoolEnabled:bool=@n24, Translation="0,0,0", Rotation="25,5,0", Scale="1,1,1") -> (Pose) {
    alphaInputType=Bool
    rotationMode=BMM_Additive
    rotationSpace=BCS_BoneSpace
    scaleMode=BMM_Additive
    scaleSpace=BCS_BoneSpace
    translationMode=BMM_Additive
    translationSpace=BCS_BoneSpace
    alphaBoolBlend {blendInTime=0.1, blendOption=Linear, blendOutTime=0.5}
    boneToModify {boneName=dyn_roincloth_bk_r_02}
    rotation {pitch=25, yaw=5}
   }
   n31 = Get bIsSliding()
   n32 = "Transform (Modify) Bone - Bone: dyn_loincloth_bk_l_MSTR"!(ComponentPose=@n30, bAlphaBoolEnabled:bool=@n31, Translation="0,0,0", Rotation="90,0.000,0.000", Scale="1,1,1") -> (Pose) {
    alphaInputType=Bool
    rotationMode=BMM_Additive
    rotationSpace=BCS_BoneSpace
    alphaBoolBlend {blendInTime=0.1, blendOption=Linear, blendOutTime=0.1}
    alphaScaleBiasClamp {bInterpResult=true}
    boneToModify {boneName=dyn_loincloth_bk_l_MSTR}
    rotation {pitch=90}
   }
   n33 = "Transform (Modify) Bone - Bone: dyn_loincloth_bk_MSTR"!(ComponentPose=@n32, bAlphaBoolEnabled:bool=@n31, Translation="0,0,0", Rotation="90,-0.000,0.000", Scale="1,1,1") -> (Pose) {
    alphaInputType=Bool
    rotationMode=BMM_Additive
    rotationSpace=BCS_BoneSpace
    alphaBoolBlend {blendInTime=0.1, blendOption=Linear, blendOutTime=0.1}
    alphaScaleBiasClamp {bInterpResult=true}
    boneToModify {boneName=dyn_loincloth_bk_MSTR}
    rotation {pitch=90}
   }
   n34 = "Transform (Modify) Bone - Bone: dyn_roincloth_bk_r_MSTR"!(ComponentPose=@n33, bAlphaBoolEnabled:bool=@n31, Translation="0,0,0", Rotation="90,15,0", Scale="1,1,1") -> (Pose) {
    alphaInputType=Bool
    rotationMode=BMM_Additive
    rotationSpace=BCS_BoneSpace
    alphaBoolBlend {blendInTime=0.1, blendOption=Linear, blendOutTime=0.1}
    alphaScaleBiasClamp {bInterpResult=true}
    boneToModify {boneName=dyn_roincloth_bk_r_MSTR}
    rotation {pitch=90, yaw=15}
   }
   n35 = "Transform (Modify) Bone - Bone: dyn_l_big_shldr_pad"!(ComponentPose=@n34, bAlphaBoolEnabled:bool=Get bIsSkyDiving(), Translation="10,0,0", Rotation="-5,0.000,-20.000", Scale="1,1,1") -> (Pose) {
    alphaInputType=Bool
    rotationMode=BMM_Additive
    rotationSpace=BCS_BoneSpace
    scaleMode=BMM_Additive
    scaleSpace=BCS_BoneSpace
    translationMode=BMM_Additive
    translationSpace=BCS_BoneSpace
    alphaBoolBlend {blendInTime=0.1, blendOption=Linear, blendOutTime=0.1}
    boneToModify {boneName=dyn_l_big_shldr_pad}
    rotation {pitch=-5, roll=-20}
    translation {x=10}
   }
   n36 = "Constraint - dyn_shoulderpad_l (P)"!(ComponentPose=@n35, bAlphaBoolEnabled:bool=True, ConstraintWeights_0=1) -> (Pose) {
    alphaInputType=Bool
    alphaBoolBlend {blendOption=Linear}
    boneToModify {boneName=dyn_shoulderpad_l}
    constraintSetup [{targetBone {boneName=upperarm_twist_01_l}}]
    constraintWeights [1]
   }
   n37 = Get skydive shldrs() -> (skydive shldrs:bool)
   n38 = "Transform (Modify) Bone - Bone: dyn_shoulderpad_r"!(ComponentPose=@n36, bAlphaBoolEnabled:bool=@n37, Translation="3,0,-3", Rotation="10,0,0", Scale="1,1,1") -> (Pose) {
    alphaInputType=Bool
    rotationMode=BMM_Additive
    rotationSpace=BCS_BoneSpace
    scaleMode=BMM_Additive
    scaleSpace=BCS_BoneSpace
    translationMode=BMM_Additive
    translationSpace=BCS_BoneSpace
    alphaBoolBlend {blendInTime=0.1, blendOption=Linear, blendOutTime=0.1}
    boneToModify {boneName=dyn_shoulderpad_r}
    rotation {pitch=10}
    translation {x=3, z=-3}
   }
   n39 = "Transform (Modify) Bone - Bone: dyn_shoulderpad_l"!(ComponentPose=@n38, bAlphaBoolEnabled:bool=@n37, Translation="5,0,4", Rotation="0,0,0", Scale="1,1,1") -> (Pose) {
    alphaInputType=Bool
    rotationMode=BMM_Additive
    rotationSpace=BCS_BoneSpace
    scaleMode=BMM_Additive
    scaleSpace=BCS_BoneSpace
    translationMode=BMM_Additive
    translationSpace=BCS_BoneSpace
    alphaBoolBlend {blendInTime=0.1, blendOption=Linear, blendOutTime=0.1}
    boneToModify {boneName=dyn_shoulderpad_l}
    translation {x=5, z=4}
   }
   n40 = "Transform (Modify) Bone - Bone: dyn_l_big_shldr_pad"(ComponentPose=@n39, bAlphaBoolEnabled=@n37, Translation="0,-5,0", Rotation="-5,0.000,0", Scale="1,1,1") {
    alphaInputType=Bool
    rotationMode=BMM_Additive
    rotationSpace=BCS_BoneSpace
    scaleMode=BMM_Additive
    scaleSpace=BCS_BoneSpace
    translationMode=BMM_Additive
    translationSpace=BCS_BoneSpace
    alphaBoolBlend {blendInTime=0.1, blendOption=Linear, blendOutTime=0.1}
    boneToModify {boneName=dyn_l_big_shldr_pad}
    rotation {pitch=-5}
    translation {y=-5}
   }
   n41 = "Transform (Modify) Bone - Bone: dyn_shoulderpad_r"(
    ComponentPose=@n40
    bAlphaBoolEnabled=Get skydive shldrs_1() -> (skydive shldrs_1:bool)
    Translation="0,0,-2.5"
    Rotation="10,0,0"
    Scale="1,1,1"
   ) {
    alphaInputType=Bool
    rotationMode=BMM_Additive
    rotationSpace=BCS_BoneSpace
    scaleMode=BMM_Additive
    scaleSpace=BCS_BoneSpace
    translationMode=BMM_Additive
    translationSpace=BCS_BoneSpace
    alphaBoolBlend {blendInTime=0.5, blendOption=Linear, blendOutTime=0.1}
    boneToModify {boneName=dyn_shoulderpad_r}
    rotation {pitch=10}
    translation {z=-2.5}
   }
   n42 = "Transform (Modify) Bone - Bone: dyn_l_big_shldr_pad"!(ComponentPose=@n41, Alpha:float=Get Arm L rot() -> (Arm L rot:double), Translation="-7,0,0", Rotation="0,0,0", Scale="1,1,1") -> (Pose) {
    rotationMode=BMM_Additive
    rotationSpace=BCS_BoneSpace
    scaleMode=BMM_Additive
    scaleSpace=BCS_BoneSpace
    translationMode=BMM_Additive
    translationSpace=BCS_BoneSpace
    alphaBoolBlend {blendInTime=0.1, blendOption=Linear, blendOutTime=0.1}
    alphaScaleBiasClamp {bInterpResult=true, interpSpeedIncreasing=20}
    boneToModify {boneName=dyn_l_big_shldr_pad}
    translation {x=-7}
   }
   n43 = "Transform (Modify) Bone - Bone: dyn_neckguard_fix"!(ComponentPose=@n42, bAlphaBoolEnabled:bool=Get neck() -> (neck:bool), Translation="-5,0,0", Rotation="0,0,0", Scale="0.85,1,1") -> (Pose) {
    alphaInputType=Bool
    rotationMode=BMM_Additive
    rotationSpace=BCS_BoneSpace
    scaleMode=BMM_Additive
    scaleSpace=BCS_BoneSpace
    translationMode=BMM_Additive
    translationSpace=BCS_BoneSpace
    alphaBoolBlend {blendInTime=0.1, blendOption=Linear, blendOutTime=0.1}
    boneToModify {boneName=dyn_neckguard_fix}
    scale {x=0.85}
    translation {x=-5}
   }
   Save cached pose 'PreSim'(Pose=Component To Local(ComponentPose=@n43)) {cacheName=PreSim}
  ]}
 EventGraph {nodes [
   PlayerMovement(AnimInstance=Self-Reference(), XDotProductCoefficient=1, YDotProductCoefficient=1, ZDotProductCoefficient=1)
   > Set FWD Move(FWD Move:double=0) -> (Output_Get:double)
   Event BlueprintUpdateAnimation()
   > n1 = Sequence!(execute) -> (then_0, then_1, then_2:Exec, then_3:Exec, then_4:Exec, then_5:Exec, then_6:Exec, then_7:Exec)
   n2 = Get bIsMoving2D()
   n3 = Branch(execute=@n1.then_1, Condition=AND Boolean(A=Get bIsCrouching(), B=NOT Boolean(A=@n2)))
   > Set still_crouch(still_crouch:bool=true) -> (Output_Get:bool)
   Set still_crouch(execute=@n3.else, still_crouch=false)
   n4 = Cast To FortPlayerAnimInstance(execute=@n1.then_2, Object=GetAnimInstance(self=GetOwnerSkeletalMeshComponent()))
   > Set Rifle(Rifle:bool=false) -> (Output_Get:bool)
   > Set pistol(pistol:bool=false) -> (Output_Get:bool)
   > Set pickaxe(pickaxe:bool=false) -> (Output_Get:bool)
   > n5 = Switch on EFortWeaponCoreAnimation!(execute, Selection:EFortWeaponCoreAnimation Enum=Get WeaponCoreAnim(self=@n4)) -> (
    Melee:Exec
    Pistol:Exec
    Shotgun:Exec
    PaperBlueprint:Exec
    Rifle:Exec
    MeleeOneHand:Exec
    MachinePistol:Exec
    RocketLauncher:Exec
    GrenadeLauncher:Exec
    GoingCommando:Exec
    AssaultRifle:Exec
    TacticalShotgun:Exec
    SniperRifle:Exec
    TrapPlacement:Exec
    ShoulderLauncher:Exec
    AbilityDecoTool:Exec
    Crossbow:Exec
    C4:Exec
    RemoteControl:Exec
    DualWield:Exec
    AR_BullPup:Exec
    AR_ForwardGrip:Exec
    MedPackPaddles:Exec
    SMG_P90:Exec
    AR_DrumGun:Exec
    Consumable_Small:Exec
    Consumable_Large:Exec
    Balloon:Exec
    MountedTurret:Exec
    CreativeTool:Exec
    ExplosiveBow:Exec
    AshtonIndigo:Exec
    AshtonChicago:Exec
    MeleeDualWield:Exec
    Unarmed:Exec
   )
   Branch(execute=[@n5.AssaultRifle, @n5.Crossbow, @n5.GrenadeLauncher, @n5.Rifle, @n5.RocketLauncher, @n5.Shotgun, @n5.SniperRifle, @n5.TacticalShotgun])
   > Set Rifle(Rifle=true)
   > Set movement(movement:double=0) -> (Output_Get:double)
   Branch(execute=[@n5.MachinePistol, @n5.Pistol])
   > Set pistol(pistol=true)
   Set pickaxe(execute=@n5.Melee, pickaxe=true)
   > Set movement(movement=1)
   Set pickaxe(execute=@n5.MeleeDualWield, pickaxe=true)
   > Set movement(movement=1)
   Branch(execute=@n1.then_3, Condition=Get pistol() -> (pistol:bool))
   > n6 = Sequence!(execute) -> (then_0, then_1)
   > n7 = Branch(Condition=OR Boolean!(
     A=Get bIsRelaxedLevel1() -> (bIsRelaxedLevel1:bool)
     B=Get bIsRelaxedLevel2() -> (bIsRelaxedLevel2:bool)
     C:bool=Get bIsSprinting()
     D:bool=Get bIsTacticalSprinting()
     E:bool=Get bIsSwimSprinting() -> (bIsSwimSprinting:bool)
    ) -> (ReturnValue))
   Set movement(execute=[@n7.then, @n8.else], movement=1)
   n8 = Branch(execute=@n7.else, Condition=NOT Boolean(A=Get bIsSkyDiving()))
   Set movement(execute=[@n8.then, @n9.then], movement=0)
   n9 = Branch(execute=@n6.then_1, Condition=Get bIsTargeting())
   n10 = Get bIsParachuteOpened()
   n11 = Get has_hand_glider() -> (has_hand_glider:bool)
   n12 = Branch(execute=@n1.then_5, Condition=OR Boolean!(A=Get bIsSkydiveFloating(), B=AND Boolean(A=@n11, B=@n10)) -> (ReturnValue))
   > Set skydive shldrs(skydive shldrs:bool=true) -> (Output_Get:bool)
   Set skydive shldrs(execute=@n12.else, skydive shldrs=false)
   Branch(execute=@n1.then_6, Condition=Get bIsCrouching())
   > n13 = Sequence()
   > n14 = Branch(Condition=OR Boolean(A=Get bIsRelaxedLevel1(), B=Get bIsRelaxedLevel2()))
   > Set Arm L rot(Arm L rot:double=0) -> (Output_Get:double)
   Set Arm L rot(
    execute=[@n14.else, @n15.then]
    Arm L rot=GetAlphaFromReferenceTransformDelta(SkeletalMeshComponent=Get Skel_Mesh(), BoneName=upperarm_l, BaseName=spine_05, InputRange=(X=40,Y=15), TwistAxis="1,0,0")
   )
   n15 = Branch(execute=@n13.then_1, Condition=Get bIsTargeting())
   n16 = Branch(execute=@n1.then_7, Condition=AND Boolean(A=Get has_hand_glider(), B=Get bIsParachuteOpened()))
   > Set skydive shldrs_1(skydive shldrs_1:bool=true) -> (Output_Get:bool)
   Set skydive shldrs_1(execute=@n16.else, skydive shldrs_1=false)
   n17 = Gravity Override State Function DEPRECATED(
    execute=@n1.then_0
    CharacterPartAnimInstance=Self-Reference()
    GameTimeInSeconds=GetGameTimeInSeconds()
    GravityOverrideParams=Get Gravity Override Params()
   )
   > Set GravityOverride(GravityOverride=@n17)
   > Set neck(neck:bool=true) -> (Output_Get:bool)
   Set has_hand_glider(execute=@n1.then_4, has_hand_glider:bool=false) -> (Output_Get:bool)
   n18 = Get fort_player() -> (fort_player:Fort Player Anim Instance Object Reference)
   > Switch on EGliderType(Selection=Break Fort Anim Input Player Glider Anim Asset!(FortAnimInput_PlayerGliderAnimAsset:"Fort Anim Input Player Glider Anim Asset Structure (by ref)"=Get CurrentGliderAnimSet!(self:Fort Player Anim Instance Object Reference=@n18) -> (CurrentGliderAnimSet:Fort Anim Input Player Glider Anim Asset Structure)) -> (
     Default_Base_BS:Blend Space Object Reference
     Default_BodyAdditive_MaleMedium_BS:Blend Space Object Reference
     Default_BodyAdditive_MaleLarge_BS:Blend Space Object Reference
     Default_BodyAdditive_FemaleSmall_BS:Blend Space Object Reference
     Default_BodyAdditive_FemaleMedium_BS:Blend Space Object Reference
     Default_BodyAdditive_FemaleLarge_BS:Blend Space Object Reference
     Default_TurnAdditive_MaleMedium_BS:Blend Space Object Reference
     Default_TurnAdditive_MaleLarge_BS:Blend Space Object Reference
     Default_TurnAdditive_FemaleSmall_BS:Blend Space Object Reference
     Default_TurnAdditive_FemaleMedium_BS:Blend Space Object Reference
     Default_TurnAdditive_FemaleLarge_BS:Blend Space Object Reference
     Into_Base_BS:Blend Space Object Reference
     Into_Base_MaleMedium_BS:Blend Space Object Reference
     Into_BodyAdditive_MaleMedium_BS:Blend Space Object Reference
     Into_BodyAdditive_MaleLarge_BS:Blend Space Object Reference
     Into_BodyAdditive_FemaleSmall_BS:Blend Space Object Reference
     Into_BodyAdditive_FemaleMedium_BS:Blend Space Object Reference
     Into_BodyAdditive_FemaleLarge_BS:Blend Space Object Reference
     Into_TurnAdditive_MaleMedium_BS:Blend Space Object Reference
     Into_TurnAdditive_MaleLarge_BS:Blend Space Object Reference
     Into_TurnAdditive_FemaleSmall_BS:Blend Space Object Reference
     Into_TurnAdditive_FemaleMedium_BS:Blend Space Object Reference
     Into_TurnAdditive_FemaleLarge_BS:Blend Space Object Reference
     Lean_MaleMedium_BS:Blend Space Object Reference
     Lean_MaleLarge_BS:Blend Space Object Reference
     Lean_FemaleSmall_BS:Blend Space Object Reference
     Lean_FemaleMedium_BS:Blend Space Object Reference
     Lean_FemaleLarge_BS:Blend Space Object Reference
     LeanAdditive_Center_MaleMedium_Pose:Anim Sequence Object Reference
     LeanAdditive_Center_MaleLarge_Pose:Anim Sequence Object Reference
     LeanAdditive_Center_FemaleSmall_Pose:Anim Sequence Object Reference
     LeanAdditive_Center_FemaleMedium_Pose:Anim Sequence Object Reference
     LeanAdditive_Center_FemaleLarge_Pose:Anim Sequence Object Reference
     LeanAdditive_Into_BS:Blend Space Object Reference
     LeanAdditive_ForwardInto_Anim:Anim Sequence Object Reference
     LeanAdditive_ForwardInto_FromDeploy_Anim:Anim Sequence Object Reference
     LeanAdditive_BackInto_Anim:Anim Sequence Object Reference
     LeanAdditive_LeftInto_Anim:Anim Sequence Object Reference
     LeanAdditive_RightInto_Anim:Anim Sequence Object Reference
     LeanAdditive_ForwardOut_Anim:Anim Sequence Object Reference
     LeanAdditive_BackOut_Anim:Anim Sequence Object Reference
     LeanAdditive_LeftOut_Anim:Anim Sequence Object Reference
     LeanAdditive_RightOut_Anim:Anim Sequence Object Reference
     ToGlide_BS:Blend Space Object Reference
     ToGlide_Lean_BS:Blend Space Object Reference
     ToDive_BS:Blend Space Object Reference
     ToDive_Lean_BS:Blend Space Object Reference
     Dive_WeaponR_Additive_BS:Blend Space Object Reference
     Glide_WeaponR_Additive_Anim:Anim Sequence Object Reference
     GenericAdditive_Male_BS:Blend Space Object Reference
     GenericAdditive_Female_BS:Blend Space Object Reference
     RootModPitchMin:float=~
     RootModPitchMax:float=~
     RootModYOffsetMin:float=~
     RootModYOffsetMax:float=~
     PelvisModVertical:Pelvis Mod Vertical Input Spring Structure
     PelvisModLateral:Pelvis Mod Lateral Input Spring Structure
     PelvisModBankAngle:Pelvis Mod Bank Angle Input Spring Structure
     ThighModLegAngle:Thigh Mod Leg Angle Input Structure
     ThighModLegPitch:Thigh Mod Leg Pitch Spring Structure
     ThighModLegBank:Thigh Mod Leg Bank Spring Structure
     ThighModLeftLegDrag:Thigh Mod Left Leg Drag Structure
     PlayerGliderType:EGliderType Enum
     bEnableSpringMods:bool=~
     bAllowPlayerDeployRootMod:bool=~
     bUseSurfStyle:bool=~
     bUseLeanPlayRateAdjust:bool=~
     bUseForwardAs_Y_BS:bool=~
     bUseDeployCurveForIntoPlayRateAdjust:bool=~
     LeanForwardValue:float=~
     IdleValue:float=~
     LeanBackValue:float=~
     IdleInterpToSpeed:float=~
     LeanForwardInterpToSpeed:float=~
     LeanBackInterpToSpeed:float=~
    ) {is=title K2Node_BreakStruct}.PlayerGliderType)
   > Set has_hand_glider(has_hand_glider=true)
   n19 = Event OnBeginFalling()
   Set jumping(execute=[@n19.then, @n20.then], jumping:double=1) -> (Output_Get:double)
   n20 = Event OnBeginJumping()
   n21 = Event OnEndJumping()
   Set jumping(execute=[@n21.then, @n22.then], jumping=0)
   n22 = Event OnEndFalling()
   Event BlueprintInitializeAnimation()
   n23 = GetOwnerSkeletalMeshComponent()
   > Set Skel_Mesh(Skel_Mesh=@n23)
   > Set Body_Mesh(Body_Mesh=@n23)
   > Set fort_player(fort_player:Fort Player Anim Instance Object Reference=GetOwnerFortPlayerAnimInstance!(self:Custom Character Part Anim Instance Object Reference) -> (ReturnValue:Fort Player Anim Instance Object Reference)) -> (Output_Get:Fort Player Anim Instance Object Reference)
   n24 = Branch(Condition=Get bIsSkyDiving())
   > Set NewSkydiving(NewSkydiving:bool=true) -> (Output_Get:bool)
   Set NewSkydiving(execute=@n24.else, NewSkydiving=false)
   n25 = Branch(Condition=Get bIsCrouching())
   > Set NewCrouching(NewCrouching:bool=true) -> (Output_Get:bool)
   Set NewCrouching(execute=@n25.else, NewCrouching=false)
   n26 = Branch(Condition=Get bIsSliding())
   > Set NewSliding(NewSliding:bool=true) -> (Output_Get:bool)
   Set NewSliding(execute=@n26.else, NewSliding=false)
   n27 = Branch(Condition=OR Boolean!(A=Get bIsAccelerating2D(), B=Get bIsPlayingEmote(), C:bool=Get bIsSkyDiving()) -> (ReturnValue))
   > Set RootSim(RootSim:bool=true) -> (Output_Get:bool)
   Set RootSim(execute=@n27.else, RootSim=false)
   n28 = Branch(Condition=Get bIsTargeting())
   > Set movement(movement=0)
   n29 = Branch(
    execute=@n28.else
    Condition=OR Boolean!(A=Get bIsSprinting(), B=Get bIsTacticalSprinting(), C=Get bIsSwimSprinting(), D:bool, E:bool) -> (ReturnValue)
   )
   > Set movement(movement=1)
   Set movement(execute=@n29.else, movement=0)
   NOT Boolean(A=false)
   AND Boolean(A=false, B=false)
   Get bIsRelaxedLevel2()
   Get bIsRelaxedLevel1()
  ]}
]
variables [
 Arm L rot {type=number}
 Body_Mesh {type=SkeletalMeshComponent}
 fort_player {type=FortPlayerAnimInstance, default=null}
 FWD Move {type=number}
 Gravity Override Params {type=GravityOverrideParamsStruct}
 GravityOverride {type=Vector}
 has_hand_glider {type=boolean}
 jumping {type=number}
 movement {type=number}
 neck {type=boolean}
 NewCrouching {type=boolean}
 NewSkydiving {type=boolean}
 NewSliding {type=boolean}
 pickaxe {type=boolean}
 pistol {type=boolean}
 Rifle {type=boolean}
 RootSim {type=boolean}
 Skel_Mesh {type=SkeletalMeshComponent}
 skydive shldrs {type=boolean}
 skydive shldrs_1 {type=boolean}
 still_crouch {type=boolean}
]
